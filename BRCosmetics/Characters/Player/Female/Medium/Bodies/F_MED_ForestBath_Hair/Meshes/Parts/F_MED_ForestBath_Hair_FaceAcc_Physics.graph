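# Physics asset for the FaceAcc part of the ForestBath_Hair female medium cosmetic, covering front and side hair plus neck-piece bones.
# Each dyn_ capsule hangs off a tiny base_ capsule through tight swing and twist limits.
# Larger head, clavicle and upper-arm capsules cover the body.
# Hair bases link to head; neck bases, head and clavicles link to spine_05, which has no body here, all with free rotation.
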PhysicsAsset /BRCosmetics/Characters/Player/Female/Medium/Bodies/F_MED_ForestBath_Hair/Meshes/Parts/F_MED_ForestBath_Hair_FaceAcc_Physics
bodies {
 base_hair_front_l [{shapeType=Capsule, length=0.505, radius=0.505, shapeName=base_hair_front_l_capsule, rotation=<rot 0 -90 0>}]
 base_hair_front_r [{shapeType=Capsule, length=0.505, radius=0.505, shapeName=base_hair_front_r_capsule, rotation=<rot 0 -90 0>}]
 base_hair_side_l [{shapeType=Capsule, length=0.505, radius=0.505, shapeName=base_hair_side_l_capsule, rotation=<rot 0 -90 0>}]
 base_hair_side_r [{shapeType=Capsule, length=0.505, radius=0.505, shapeName=base_hair_side_r_capsule, rotation=<rot 0 -90 0>}]
 base_neck_back [{shapeType=Capsule, length=0.505, radius=0.505, shapeName=base_neck_back_capsule, rotation=<rot 0 -90 0>}]
 base_neck_front_l [{shapeType=Capsule, length=0.505, radius=0.505, shapeName=base_neck_front_l_capsule, rotation=<rot 0 -90 0>}]
 base_neck_front_r [{shapeType=Capsule, length=0.505, radius=0.505, shapeName=base_neck_front_r_capsule, rotation=<rot 0 -90 0>}]
 base_neck_mid_l [{shapeType=Capsule, length=0.505, radius=0.505, shapeName=base_neck_mid_l_capsule, rotation=<rot 0 -90 0>}]
 base_neck_mid_r [{shapeType=Capsule, length=0.505, radius=0.505, shapeName=base_neck_mid_r_capsule, rotation=<rot 0 -90 0>}]
 base_neck_midback_r [{shapeType=Capsule, length=0.505, radius=0.505, shapeName=base_neck_midback_r_capsule, rotation=<rot 0 -90 0>}]
 clavicle_l [{shapeType=Capsule, length=4, radius=7.5, shapeName=clavicle_l_capsule, center=<5.7043 0 -0.9159>, rotation=<rot 90 -87.106 -87.106>}]
 clavicle_r [{shapeType=Capsule, length=5, radius=6.5, shapeName=clavicle_l_capsule, center=<-6 0 1>, rotation=<rot 85 -0.0002 -0.0002>}]
 dyn_hair_front_l [{shapeType=Capsule, length=6, radius=1.3, shapeName=dyn_hair_front_l_capsule, center=<3 -1 0>, rotation=<rot 90 0 5>}]
 dyn_hair_front_r [{shapeType=Capsule, length=7, radius=1.3, shapeName=dyn_hair_front_l_capsule, center=<3.5 -0.9 1>, rotation=<rot 90 0 3>}]
 dyn_hair_side_l [{shapeType=Capsule, length=12, radius=1.3, shapeName=dyn_hair_front_l_capsule, center=<5.5 0.3 -1.6>, rotation=<rot 90 0 0>}]
 dyn_hair_side_r [{shapeType=Capsule, length=3, radius=2, shapeName=dyn_hair_front_r_capsule, center=<8.6 1.5 2>, rotation=<rot 0 0 10>}]
 dyn_neck_back [{shapeType=Capsule, length=6, radius=1.5, shapeName=dyn_neck_back_capsule, center=<2.3 0.3 0>}]
 dyn_neck_front_l [{shapeType=Capsule, length=11, radius=0.6, shapeName=dyn_neck_front_l_capsule, center=<3 -4.2 0>, rotation=<rot 0 -18 90>}]
 dyn_neck_front_r [{shapeType=Capsule, length=11, radius=0.6, shapeName=dyn_neck_front_l_capsule, center=<2.8 -4.2 0>, rotation=<rot 0 -17 90>}]
 dyn_neck_mid_l [{shapeType=Capsule, length=11, radius=0.6, shapeName=dyn_neck_front_l_capsule, center=<2 -1.85 0.5>, rotation=<rot 0 5 92>}]
 dyn_neck_mid_r [{shapeType=Capsule, length=9, radius=0.6, shapeName=dyn_neck_front_l_capsule, center=<1.7 -1.7 -0.25>, rotation=<rot 0 0 90>}]
 dyn_neck_midback_r [{shapeType=Capsule, length=4, radius=1.5, shapeName=dyn_neck_midback_r_capsule, center=<1.8 0 0>, rotation=<rot 0 0 90>}]
 head [{shapeType=Capsule, length=8, radius=8, shapeName=head_capsule, center=<4 0 0.0067>, rotation=<rot 0 58.2958 -89.8606>}]
 upperarm_l [{shapeType=Capsule, length=16, radius=7, shapeName=upperarm_l_capsule, center=<11 0 -0.5>, rotation=<rot 90 -90.9147 -90.9147>}]
 upperarm_r [{shapeType=Capsule, length=16, radius=6.5, shapeName=upperarm_r_capsule, center=<-12 0 0>, rotation=<rot 85 -179.9999 -179.9999>}]
}
constraints [
 {bone1Name=dyn_hair_front_l, bone2Name=base_hair_front_l, swing1LimitDegrees=15, swing2LimitDegrees=20, twistLimitDegrees=1}
 {bone1Name=dyn_hair_front_r, bone2Name=base_hair_front_r, swing1LimitDegrees=15, swing2LimitDegrees=20, twistLimitDegrees=1}
 {bone1Name=dyn_hair_side_l, bone2Name=base_hair_side_l, swing1LimitDegrees=8, swing2LimitDegrees=20, twistLimitDegrees=1}
 {bone1Name=dyn_hair_side_r, bone2Name=base_hair_side_r, swing1LimitDegrees=8, swing2LimitDegrees=20, twistLimitDegrees=1}
 {
  bone1Name=upperarm_l
  bone2Name=clavicle_l
  swing1LimitDegrees=45
  swing1Motion=Free
  swing2LimitDegrees=45
  swing2Motion=Free
  twistLimitDegrees=45
  twistMotion=Free
 }
 {
  bone1Name=upperarm_r
  bone2Name=clavicle_r
  swing1LimitDegrees=45
  swing1Motion=Free
  swing2LimitDegrees=45
  swing2Motion=Free
  twistLimitDegrees=45
  twistMotion=Free
 }
 {
  bone1Name=base_hair_front_l
  bone2Name=head
  swing1LimitDegrees=45
  swing1Motion=Free
  swing2LimitDegrees=45
  swing2Motion=Free
  twistLimitDegrees=45
  twistMotion=Free
 }
 {
  bone1Name=base_hair_front_r
  bone2Name=head
  swing1LimitDegrees=45
  swing1Motion=Free
  swing2LimitDegrees=45
  swing2Motion=Free
  twistLimitDegrees=45
  twistMotion=Free
 }
 {
  bone1Name=base_hair_side_l
  bone2Name=head
  swing1LimitDegrees=45
  swing1Motion=Free
  swing2LimitDegrees=45
  swing2Motion=Free
  twistLimitDegrees=45
  twistMotion=Free
 }
 {
  bone1Name=base_hair_side_r
  bone2Name=head
  swing1LimitDegrees=45
  swing1Motion=Free
  swing2LimitDegrees=45
  swing2Motion=Free
  twistLimitDegrees=45
  twistMotion=Free
 }
 {
  bone1Name=clavicle_l
  bone2Name=spine_05
  swing1LimitDegrees=45
  swing1Motion=Free
  swing2LimitDegrees=45
  swing2Motion=Free
  twistLimitDegrees=45
  twistMotion=Free
 }
 {
  bone1Name=clavicle_r
  bone2Name=spine_05
  swing1LimitDegrees=45
  swing1Motion=Free
  swing2LimitDegrees=45
  swing2Motion=Free
  twistLimitDegrees=45
  twistMotion=Free
 }
 {
  bone1Name=head
  bone2Name=spine_05
  swing1LimitDegrees=45
  swing1Motion=Free
  swing2LimitDegrees=45
  swing2Motion=Free
  twistLimitDegrees=45
  twistMotion=Free
 }
 {bone1Name=dyn_neck_back, bone2Name=base_neck_back, swing1LimitDegrees=15, swing2LimitDegrees=5, twistLimitDegrees=1}
 {bone1Name=dyn_neck_front_l, bone2Name=base_neck_front_l, swing1LimitDegrees=0, swing2LimitDegrees=20, twistLimitDegrees=1}
 {bone1Name=dyn_neck_front_r, bone2Name=base_neck_front_r, swing1LimitDegrees=0, swing2LimitDegrees=20, twistLimitDegrees=1}
 {bone1Name=dyn_neck_mid_l, bone2Name=base_neck_mid_l, swing1LimitDegrees=5, swing2LimitDegrees=15, twistLimitDegrees=1}
 {bone1Name=dyn_neck_mid_r, bone2Name=base_neck_mid_r, swing1LimitDegrees=5, swing2LimitDegrees=15, twistLimitDegrees=1}
 {bone1Name=dyn_neck_midback_r, bone2Name=base_neck_midback_r, swing1LimitDegrees=5, swing2LimitDegrees=15, twistLimitDegrees=1}
 {
  bone1Name=base_neck_back
  bone2Name=spine_05
  swing1LimitDegrees=45
  swing1Motion=Free
  swing2LimitDegrees=45
  swing2Motion=Free
  twistLimitDegrees=45
  twistMotion=Free
 }
 {
  bone1Name=base_neck_front_l
  bone2Name=spine_05
  swing1LimitDegrees=45
  swing1Motion=Free
  swing2LimitDegrees=45
  swing2Motion=Free
  twistLimitDegrees=45
  twistMotion=Free
 }
 {
  bone1Name=base_neck_front_r
  bone2Name=spine_05
  swing1LimitDegrees=45
  swing1Motion=Free
  swing2LimitDegrees=45
  swing2Motion=Free
  twistLimitDegrees=45
  twistMotion=Free
 }
 {
  bone1Name=base_neck_mid_l
  bone2Name=spine_05
  swing1LimitDegrees=45
  swing1Motion=Free
  swing2LimitDegrees=45
  swing2Motion=Free
  twistLimitDegrees=45
  twistMotion=Free
 }
 {
  bone1Name=base_neck_mid_r
  bone2Name=spine_05
  swing1LimitDegrees=45
  swing1Motion=Free
  swing2LimitDegrees=45
  swing2Motion=Free
  twistLimitDegrees=45
  twistMotion=Free
 }
 {
  bone1Name=base_neck_midback_r
  bone2Name=spine_05
  swing1LimitDegrees=45
  swing1Motion=Free
  swing2LimitDegrees=45
  swing2Motion=Free
  twistLimitDegrees=45
  twistMotion=Free
 }
]
